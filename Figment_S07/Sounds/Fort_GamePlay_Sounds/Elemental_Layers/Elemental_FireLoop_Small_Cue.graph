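# Sound cue for the looping fire layer of a small elemental enemy, played at half volume in the non-critical enemy sound class.
# A Random node picks one of two equally weighted looping fire waves, which passes through a Modulator at the root.
# It uses small monster elemental attenuation, and its concurrency override caps it at two instances, stopping the farthest before preventing new ones.
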SoundCue /Figment_S07/Sounds/Fort_GamePlay_Sounds/Elemental_Layers/Elemental_FireLoop_Small_Cue
root=Modulator_0
nodes [
 {id=Modulator_0, children [Random_0]}
 {id=Random_0, children [WavePlayer_4,WavePlayer_5], properties {weights [1,1]}}
 {id=WavePlayer_4, properties {bLooping=true, soundWaveAssetPtr=&/enemy_fire_loop_02b}}
 {id=WavePlayer_5, properties {bLooping=true, soundWaveAssetPtr=&/enemy_fire_loop_03a}}
]
properties {
 attenuationSettings=/Game/Sounds/Attenuation/Fort_Monster_Elemental_Small_Att
 bOverrideConcurrency=true
 maxDistance=1100
 soundClassObject=/Game/Sounds/FortSoundClassesAndModes/Fort_Enemy_NonCritical
 volumeMultiplier=0.5
 concurrencyOverrides {maxCount=2, resolutionRule=StopFarthestThenPreventNew}
}
